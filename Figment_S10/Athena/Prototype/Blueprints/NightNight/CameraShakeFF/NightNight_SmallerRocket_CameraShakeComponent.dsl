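; Client-side component that drives a camera shake and force feedback based on proximity to its owner, apparently a NightNight event rocket.
; Each tick it starts a new shake on the local player's camera manager, overriding its Perlin rotation oscillators.
; Zero ShakeStrength zeroes them; otherwise a radius/power distance falloff updates strength and amplitude, with rumble above a threshold.
; BeginPlay zeroes ShakeStrength, so the falloff branch requires it to be set elsewhere.
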(event EventBeginPlay
  (if (not (Networking|IsServer))
    (Variables|Default|SetShakeStrength 0.0)
    (Components|Tick|SetComponentTickEnabled true)
    (Variables|Default|SetLocalPlayer (Utilities|Array|Get(acopy) (Fortnite|GetLocalFortPlayerControllers true true)))))

(event EventTick
  (params (DeltaSeconds :type Float%20%28single-precision%29))
  (bind _falloffpower (Variables|Default|GetFalloffPower))
  (bind _radius (Variables|Default|GetRadius))
  (if (not (Networking|IsServer))
    (Utilities|IsValid (Pawn|GetViewTarget (Variables|Default|GetLocalPlayer))
      (:Is%20Valid
        (if (== (Variables|Default|GetShakeStrength) 0.0)
          (bind _returnvalue (CameraShakes|StartCameraShake :self (Variables|/Script/Engine.PlayerController|GetPlayerCameraManager (Variables|Default|GetLocalPlayer)) :ShakeClass "/Figment_S10/Athena/Prototype/Blueprints/NightNight/CameraShakeFF/NightNight_SmallerRocket_CameraShake.NightNight_SmallerRocket_CameraShake_C" :Scale 1.0 :PlaySpace "CameraLocal"))
          (Variables|/Script/EngineCameras.LegacyCameraShake|SetRotOscillation (Utilities|Struct|MakeROscillator (struct :Amplitude 0.0 :Frequency 0.0 :Waveform "PerlinNoise") (struct :Amplitude 0.0 :Frequency 0.0 :Waveform "PerlinNoise") (struct :Amplitude 0.0 :Frequency 0.0 :Waveform "PerlinNoise")) (CameraShakes|ConvLegacyCameraShake _returnvalue))
          (else
            (bind _returnvalue_2 (CameraShakes|StartCameraShake :self (Variables|/Script/Engine.PlayerController|GetPlayerCameraManager (Variables|Default|GetLocalPlayer)) :ShakeClass "/Figment_S10/Athena/Prototype/Blueprints/NightNight/CameraShakeFF/NightNight_SmallerRocket_CameraShake.NightNight_SmallerRocket_CameraShake_C" :Scale 1.0 :PlaySpace "CameraLocal"))
            (Variables|Default|SetLocalPlayerCameraShake (CameraShakes|ConvLegacyCameraShake _returnvalue_2))
            (bind _distance (Class|MissionBlueprintFunctionLibrary|DistanceBetweenTwoVectors (Transformation|GetActorLocation (Pawn|GetViewTarget (Variables|Default|GetLocalPlayer))) (Transformation|GetActorLocation (Components|GetOwner))))
            (Variables|Default|SetShakeStrength (Math|Float|Power (Math|Float|MapRangeClamped _distance 0.0 _radius 1.0) _falloffpower))
            (Variables|/Script/EngineCameras.LegacyCameraShake|SetRotOscillation (Utilities|Struct|MakeROscillator (Utilities|Struct|MakeFOscillator (* (Math|Float|Power (Math|Float|MapRangeClamped _distance 0.0 _radius 1.0) _falloffpower) (Variables|Default|GetShakeStrength)) (Variables|Default|GetFrequency) "PerlinNoise") (Utilities|Struct|MakeFOscillator (* (Math|Float|Power (Math|Float|MapRangeClamped _distance 0.0 _radius 1.0) _falloffpower) (Variables|Default|GetShakeStrength)) (Variables|Default|GetFrequency) "PerlinNoise") (Utilities|Struct|MakeFOscillator (* (Math|Float|Power (Math|Float|MapRangeClamped _distance 0.0 _radius 1.0) _falloffpower) (Variables|Default|GetShakeStrength)) (Variables|Default|GetFrequency) "PerlinNoise")) (Variables|Default|GetLocalPlayerCameraShake))
            (if (not (<= (Variables|Default|GetShakeStrength) 0.2))
              (Game|Feedback|ClientPlayForceFeedback (Variables|Default|GetLocalPlayer) "/Figment_S10/Athena/Prototype/Blueprints/NightNight/CameraShakeFF/NightNight_Rocket_Flyby_ForceFeedback.NightNight_Rocket_Flyby_ForceFeedback" "None" false false false))))))))
